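; Smart object gameplay behavior in which an AI bot stops a vault break-in at a VividRazorVault bank.
; After Init caches pawn, controller and config, the bot focuses on the object, then applies a gameplay effect or plays an emote for a configured time.
; Only the effect path notifies the bank and then ends the behavior; the emote path just stops the emote.
; BuildingHealthChanged disables SmartObject components and exits when the building's health runs out.
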(event GameplayBehavior|EventOnTriggeredCharacter
  (params (Avatar :type Character%20Object%20Reference) (Config :type Gameplay%20Behavior%20Config%20Object%20Reference) (SmartObjectOwner :type Actor%20Object%20Reference))
  (bind _userpawn (Variables|Default|GetUserPawn))
  (bind _returnvalue (Ability|GetFortAbilitySystemComponent _userpawn))
  (bind _breakinconfig (Variables|Default|GetBreakInConfig))
  (bind _breakinconfig_2 (Variables|Default|GetBreakInConfig))
  (bind _output_get (Variables|Default|SetUserAvatar Avatar))
  (bind _successfulinit (CallFunction|Init _output_get Config SmartObjectOwner))
  (if _successfulinit
    (CallFunction|SetFocalPoint)
    (GameplayTasks|TaskWaitDelay 0 0.5
      (:OnFinish
        (if (Variables|/VividRazorVaultGameplay/Assets/SmartObject/SOGB_StopVaultBreakin_Config.SOGB_StopVaultBreakin_Config_C|GetUseGameplayEffectForAnimation (Variables|Default|GetBreakInConfig))
          (Utilities|IsValid _returnvalue
            (:Is%20Valid
              (GameplayEffects|ApplyGameplayEffectToSelf _returnvalue (Variables|/VividRazorVaultGameplay/Assets/SmartObject/SOGB_StopVaultBreakin_Config.SOGB_StopVaultBreakin_Config_C|GetGameplayEffectForAnimation _breakinconfig) 0.0 (GameplayEffects|MakeEffectContext _returnvalue))
              (GameplayTasks|TaskWaitDelay 0 (ScalableFloat|GetValueatLevel (Variables|/VividRazorVaultGameplay/Assets/SmartObject/SOGB_StopVaultBreakin_Config.SOGB_StopVaultBreakin_Config_C|GetAnimationTime _breakinconfig))
                (:OnFinish
                  (bind _asbga_vivid_razor_vault_bank (Utilities|Casting|CastToBGA_VividRazorVault_Bank (Variables|Default|GetSmartObjectOwner))
                    (:then
                      (Class|BGAVividRazorVaultBank|AIStoppedVaultBreakIn _asbga_vivid_razor_vault_bank))
                    (:CastFailed
                      _))
                  (GameplayBehavior|EndBehavior (Variables|Default|GetUserAvatar)))))
            (:Is%20Not%20Valid
              (GameplayBehavior|EndBehavior (Variables|Default|GetUserAvatar))))
          (else
            (Class|FortAthenaAIBotController|PlayEmote (Variables|Default|GetUserController) (Variables|/VividRazorVaultGameplay/Assets/SmartObject/SOGB_StopVaultBreakin_Config.SOGB_StopVaultBreakin_Config_C|GetEmoteID _breakinconfig_2))
            (GameplayTasks|TaskWaitDelay 0 (ScalableFloat|GetValueatLevel (Variables|/VividRazorVaultGameplay/Assets/SmartObject/SOGB_StopVaultBreakin_Config.SOGB_StopVaultBreakin_Config_C|GetAnimationTime _breakinconfig_2))
              (:OnFinish
                (Class|FortAthenaAIBotController|StopEmote (Variables|Default|GetUserController))))))))
    (else
      (bind _useravatar (Variables|Default|GetUserAvatar)
        (:then
          (GameplayBehavior|EndBehavior _useravatar))
        (:else)))))

(event GameplayBehavior|EventOnFinishedCharacter
  (params (Avatar :type Character%20Object%20Reference) (bWasInterrupted :type Boolean))
  (CallFunction|ResetFocalPoint))

(event Custom|BuildingHealthChanged
  (bind _buildingactortodestroy (Variables|Default|GetBuildingActorToDestroy))
  (if (<= (Health|GetHealth _buildingactortodestroy :!class /Script/FortniteGame.BuildingActor) 0)
    ; Force Disable Components before exiting behavior. Weird Logic were Component Stays, could be related to the prop I used.
    (for _array_element (Actor|GetComponentsByClass _buildingactortodestroy "/Script/SmartObjectsModule.SmartObjectComponent")
      (if (SmartObject|SetSmartObjectEnabled(defaultreason:Gameplay) _array_element false :!class /Script/SmartObjectsModule.SmartObjectComponent)
        _))
    (CallFunction|ResetFocalPoint)
    (GameplayBehavior|EndBehavior (Variables|Default|GetUserAvatar))))

(fn Init
  (params (Avatar :type Character%20Object%20Reference) (BehaviorConfig :type Gameplay%20Behavior%20Config%20Object%20Reference) (SmartObjectOwner :type Actor%20Object%20Reference))
  (Utilities|FlowControl|Sequence
    (:then_0
      (bind _asfort_player_pawn (Utilities|Casting|CastToFortPlayerPawn Avatar)
        (:then
          (bind _output_get (Variables|Default|SetUserPawn _asfort_player_pawn))
          (bind _asfort_athena_aibot_controller (Utilities|Casting|CastToFortAthenaAIBotController (Pawn|GetController _output_get))
            (:then
              (Variables|Default|SetUserController _asfort_athena_aibot_controller))
            (:CastFailed
              _)))
        (:CastFailed
          _))
      (LocalVariables|SetSuccessfullyInitialized false))
    (:then_1
      (bind _assogb_stop_vault_breakin_config (Utilities|Casting|CastToSOGB_StopVaultBreakin_Config BehaviorConfig))
      (Variables|Default|SetBreakInConfig _assogb_stop_vault_breakin_config))
    (:then_2
      (Variables|Default|SetSmartObjectOwner SmartObjectOwner))
    (:then_3
      (return (LocalVariables|GetSuccessfullyInitialized)))))

(fn ResetFocalPoint
  (bind _usercontroller (Variables|Default|GetUserController)
    (:then
      (AI|ClearFocus _usercontroller)
      (Class|FortAthenaAIBotController|ClearAIFocalPoint _usercontroller false))
    (:else)))

(fn SetFocalPoint
  (if (and (Utilities|IsValid (Variables|Default|GetUserController) :!class /Script/Engine.KismetSystemLibrary) (Utilities|IsValid (Variables|Default|GetSmartObjectOwner) :!class /Script/Engine.KismetSystemLibrary))
    (AI|SetFocalPoint (Variables|Default|GetUserController) (Transformation|GetActorLocation (Variables|Default|GetSmartObjectOwner)))))
